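# Niagara effect for the tropical Hera environment, named for a fish farm and presumably providing its ambient fish.
# The system contains a single emitter, Fish, with no user parameters.
# SystemState is the only listed module, and no renderers, data interfaces or dynamic inputs appear.
# The NSLW prefix suggests a lightweight emitter setup, whose internals are not shown here.
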NiagaraSystem /Game/Environments/Hera/FX/Tropical/NSLW_Env_Tropical_FishFarm
dependencies {usedDataInterfaces [], usedDynamicInputs [], usedModules [&SystemState]}
emitters {Fish {rendererClasses [], renderers []}}
summary {systemName=NSLW_Env_Tropical_FishFarm, emitters [{emitterName=Fish, rendererClasses []}], userVariables []}
