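# Niagara environment effect for a fish tank that shows schools of tropical fish along with water bubbles.
# A GPU emitter spawns randomly tinted mesh fish in a cylinder up to a user cap, steering them with a moving vortex, wind, drag, speed limits and distance field avoidance.
# A bubble emitter adds rising particles swirled by vortex and curl noise.
# Region masks use User.BoxSize, scaling fish by mask and age.
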NiagaraSystem /Game/Environments/Hera/FX/Tropical/NS_Env_TropicalFish_FishSchools_Tank
dependencies {
 usedDataInterfaces [
  &NiagaraDataInterfaceCurve
  &NiagaraDataInterfaceColorCurve
  &NiagaraDataInterfaceMeshRendererInfo
  &NiagaraDataInterfaceLandscape
 ]
 usedDynamicInputs []
 usedModules [&SystemState]
}
emitters {
 FN_WaterBubbles {
  emitterUpdateScript [
   &EmitterState()
   &SpawnRate(SpawnRate=150, Use Spawn Probability=false)
  ]
  particleSpawnScript [
   &InitializeParticle(
    Color Maximum=<rgba 0.2834 0.6114 1 1>
    Color Minimum=<rgba 0.3555 0.543 0.6406 1>
    Color Mode=::NewEnumerator2
    Lifetime Max=1
    Lifetime Min=0.5
    Lifetime Mode=::NewEnumerator1
    Sprite Size Mode=::NewEnumerator3
    Uniform Sprite Size=&FloatFromCurve(
     FloatCurve=DataInterface() {propertyValues {Curve {keys [<at 0.2 8 RCIM_Cubic>, {arriveTangent=27.6937, interpMode=RCIM_Cubic, tangentMode=RCTM_Break, time=1, value=14}]}}}
     CurveIndex=&RandomRangeFloat(Minimum=0, Maximum=2)
     Scale Curve=2
    )
    UsePositionOffset=false
    Write Material Random=true
   )
   &AddVelocity(Velocity=&RandomRangeVector(Minimum=<0 0 20>, Maximum=<1 1 100>))
   &ShapeLocation(
    Cylinder Height=200
    Cylinder Radius=425
    MaskSpawnGroup=false
    Non Uniform Scale=<1 1 0>
    Override Seed=false
    Shape Primitive=::NewEnumerator1
    Surface Only=false
    UseOwnerScale=true
   )
  ]
  particleUpdateScript [
   &ParticleState()
   &VortexForce(
    Influence Falloff Radius=400
    Origin Pull Amount=150
    Use Influence Falloff=true
    UseOriginOffset=true
    Vortex Force Amount=&RandomRangeFloat(
     Minimum=25
     Maximum=&FN_LOD_Distance_Remap(Start Distance=2000, End Distance=30, Minimum Output=0, Maximum Output=1, Clamp Output=true)
    )
    Vortex Origin Offset:Vector3f=&Lerp_Vector(
     A=<200 200 200>
     B=<-200 -200 -200>
     Alpha=&Waveform("Waveform [1]"=::NewEnumerator0, "[1] Frequency"=0.1, "[1] Amplitude Min/Max"=<0 1>)
    )
   )
   &AccelerationForce(Acceleration=&RandomRangeVector(Minimum=<0 0 2>, Maximum=<1 1 15>))
   &CurlNoiseForce(
    Noise Strength=&FloatFromCurve(
     FloatCurve=DataInterface() {propertyValues {Curve {keys [<at 0 0.05 RCIM_Cubic>, {arriveTangent=1.862, interpMode=RCIM_Cubic, tangentMode=RCTM_Break, time=1, value=1}]}}}
     Scale Curve=150
    )
    Pan Noise=false
   )
   &Drag(
    Drag=&FloatFromCurve(
     FloatCurve=DataInterface() {propertyValues {Curve {keys [
         {arriveTangent=-2.3383, interpMode=RCIM_Cubic, leaveTangent=-2.3383, tangentMode=RCTM_User, time=0, value=1}
         {interpMode=RCIM_Cubic, tangentMode=RCTM_Break, time=1, value=0}
        ]}}}
     Scale Curve=0.4
    )
    Use Linear Drag=true
    Use Rotational Drag=false
   )
   &SolveForcesAndVelocity(Clamp Velocity=false, Limit Acceleration=false)
   &DynamicMaterialParameters(
    Index 0 Param 1=&FloatFromCurve(FloatCurve=DataInterface() {propertyValues {Curve {keys [{arriveTangent=-4.0769, time=-0.0046, value=2}, <at 1.0011 0.002>]}}})
    Index 0 Param 2=&FloatFromCurve(FloatCurve=DataInterface() {propertyValues {Curve {keys [<at -0.0046 3.0507>,<at 1.0011 1>]}}})
    Index 0 Param 3=&FloatFromCurve(FloatCurve=DataInterface() {propertyValues {Curve {keys [<at -0.0046 15>,<at 1.0048 60>]}}})
    Index 0 Param 4=4
    Module.Param0WriteEnabled=true
    Module.Param1WriteEnabled=true
    Module.Param2WriteEnabled=true
    Module.Param3WriteEnabled=true
   )
   &ScaleSpriteSize(Uniform Curve Sprite Scale=DataInterface() {propertyValues {Curve {
       defaultValue=0
       keys [
        {interpMode=RCIM_Cubic, tangentMode=RCTM_User, time=0, value=0}
        {interpMode=RCIM_Cubic, tangentMode=RCTM_User, time=0.5, value=1}
        {tangentMode=RCTM_User, time=1, value=0}
       ]
      }}})
   &ScaleColor(
    Scale Alpha=&FloatFromCurve(FloatCurve=DataInterface() {propertyValues {Curve {keys [<at 0 -1>,<at 0.05 0>,<at 0.7904 0.0132>,<at 1.0011 -0.9961>]}}})
    ScaleA=true
    ScaleRGB=true
   ) {enabled=false}
   &RegionMask(Region Falloff=0.1, Region Radius:NiagaraFloat=$User.BoxSize, UseOffset=false, UseRotation=false) {module=RegionMask001}
   &KillParticlesInVolume(Invert Volume=true, Origin Offset:Vector3f=<0 0 -275>, Sphere Radius:NiagaraFloat=&Multiply_Float(A=500, B=1))
  ]
  rendererClasses [&NiagaraMeshRendererProperties, &NiagaraSpriteRendererProperties]
  renderers [
   {rendererClass=&NiagaraMeshRendererProperties, rendererIndex=0}
   {rendererClass=&NiagaraSpriteRendererProperties, rendererIndex=1}
  ]
 }
 Fish_HighEpic001 {
  simTarget=GPUComputeSim
  emitterUpdateScript [
   &EmitterState(
    Enable Distance Culling=true
    Life Cycle Mode=::NewEnumerator1
    Loop Duration=1
    Module.bApplySpawnCountScaleByDistance=false
    Module.bUseMaxDistance=true
    Module.bUseMinDistance=true
    Scalability Mode=::NewEnumerator1
    Scale Spawn Count=true
    Spawn Count Scale By Distance=DataInterface() {propertyValues {Curve {keys [<at 0 0>,<at 0.2 1>,<at 0.85 1>,<at 1 0>]}}}
    UseLoopDelay=false
   )
   &SpawnRate(
    SpawnRate=&Multiply_Float(
     A=0.25
     B=&MakeCustomFloatFromBool(
      Bool=&SetBoolByIntComparison(A=$Engine.Emitter.NumParticles, B=$User.SpawnCap, Comparison Type=::NewEnumerator0)
      True Float=0
      False Float=$User.SpawnRate
     )
    )
    Use Spawn Probability=false
   )
  ]
  particleSpawnScript [
   &InitializeParticle(
    Color=&MultiplyLinearColor(
     Linear Color=&ColorFromCurve(
      ColorCurve=DataInterface() {propertyValues {
        AlphaCurve {keys [<at 0 1>,<at 1 1>]}
        BlueCurve {keys [<at 0 0>,<at 0.9906 0>]}
        GreenCurve {keys [<at 0 0>,<at 0.9906 0.2748>]}
        RedCurve {keys [<at 0 1>,<at 0.9906 1>]}
       }}
      CurveIndex=&RandomRangeFloat(Minimum=0, Maximum=1)
     )
     Scale Factor=&MakeVector4_FromFloat(Float=&RandomRangeFloat(Minimum=2, Maximum=3))
    )
    Lifetime Max=15
    Lifetime Min=10
    Lifetime Mode=::NewEnumerator1
    Mesh Renderer Array Visibility Mode=::NewEnumerator3
    Mesh Renderer Info=DataInterface() {propertyValues {MeshRenderer=&self:Fish_HighEpic001_0.NiagaraMeshRendererProperties_1}}
    Mesh Renderer Random Seed:NiagaraInt32=&RandomRangeInt(Minimum=0, Maximum=2)
    Mesh Scale Mode=::NewEnumerator3
    Mesh Uniform Scale=&Multiply_Float(A=&RandomRangeFloat(Minimum=1, Maximum=2), B=$User.MeshScaleMult)
    UsePositionOffset=false
    Write Material Random=true
   )
   &SystemLocation(MaskSpawnGroup=false, Offset=<0 0 800>, Offset Coordinate Space=::World) {enabled=false}
   &ShapeLocation(
    Cylinder Height=50
    Cylinder Radius=$User.CylinderRadius
    MaskSpawnGroup=false
    Offset=<50 0 0>
    Offset Mode=::NewEnumerator0
    Override Seed=false
    Shape Primitive=::NewEnumerator1
    Surface Only=false
    Surface Only Band Thickness=1
    UseOwnerScale=true
   )
   &AddVelocity(Velocity=&RandomRangeVector(Minimum=<125 100 0>, Maximum=<-125 150 0>))
   &InitialMeshOrientation(
    Mesh Axis To Orient=<0 1 0>
    Mesh Orientation Mode=::NewEnumerator2
    Rotation=&RandomRangeVector(Minimum=<0 0 0.25>, Maximum=<0 0 0.25>)
    Use Rotation Vector=true
   )
   /SetVariables(Particles.VelocityClamp:Vector3f=&Multiply_Vector(A=$Particles.Velocity, B=<1 1 0>)) {isSetParameters=true}
  ]
  particleUpdateScript [
   &ParticleState()
   &DynamicMaterialParameters(
    Index 0 Param 1=&RandomRangeFloat(Minimum=0.02, Maximum=0.05)
    Index 0 Param 2=&RandomRangeFloat(Minimum=0, Maximum=0)
    Index 0 Param 3=&RandomRangeFloat(Minimum=0, Maximum=1)
    Index 0 Param 4=/Niagara/DynamicInputs/Camera/CameraFade(Camera Query:NiagaraDataInterfaceCamera=?, Fade Offset:NiagaraFloat=400, Fade Distance:NiagaraFloat=100, Use Owner Position:NiagaraBool=true)
    Index 1 Param 1=&RandomRangeFloat(Minimum=0.1, Maximum=0.2)
    Index 1 Param 2=0.25
    Index 1 Param 3=1
    Module.Index0Float0Write=true
    Module.Index0Float1Write=true
    Module.Index0Float2Write=true
    Module.Index0Float3Write=true
    Module.Index1Float0Write=true
    Module.Index1Float1Write=true
    Module.Index1Float2Write=true
    Module.Index1Float3Write=true
    Write Parameter Index 1=::NewEnumerator1
   )
   &VortexForce(
    Influence Falloff Radius=400
    Origin Pull Amount=150
    Use Influence Falloff=true
    UseOriginOffset=true
    Vortex Force Amount=&RandomRangeFloat(Minimum=2500, Maximum=2500)
    Vortex Origin Offset:Vector3f=&Lerp_Vector(
     A=<200 200 200>
     B=<-200 -200 -200>
     Alpha=&Waveform("Waveform [1]"=::NewEnumerator0, "[1] Frequency"=0.1, "[1] Amplitude Min/Max"=<0 1>)
    )
   )
   &SolveForcesAndVelocity(
    Acceleration Limit:NiagaraFloat=$User.MaxVel
    Clamp Velocity=true
    Force:Vector3f=&Multiply_Vector(A=$Transient.PhysicsForce, B=<1 1 1>)
    Limit Acceleration=true
    Speed Limit:NiagaraFloat=$User.MaxVel
    Velocity:Vector3f=&Multiply_Vector(A=$Particles.Velocity, B=<1 1 1>)
   ) {module=SolveForcesAndVelocity001}
   /Niagara/Modules/Math/AvoidDistanceFieldSurfaces_GPU(
    Avoid Nearest Surface=true
    Avoid Oncoming Surfaces=true
    Falloff Distance:NiagaraFloat=&RandomRangeFloat(Minimum=100, Maximum=150)
    Max Trace Distance:NiagaraFloat=&RandomRangeFloat(Minimum=100, Maximum=150)
    Nearest Surface Avoidance Strength=25
    Obstactle Avoidance Strength=25
   )
   &WindForce(
    Friction Mode:ENiagaraWindFrictionMode=::NewEnumerator0
    Offset By Time:Vector3f=<0 -100 0>
    Scale=1
    Surface Alignment Mode:ENiagaraWindCombingMode=::NewEnumerator0
    Turbulence Frequency Mode:ENiagaraWindTurbulenceFrequencyMode=::NewEnumerator1
    Wind Speed=&RandomRangeVector(Minimum=<-50 50 -25>, Maximum=<-50 50 25>)
    Wind Speed Scale=0.5
   )
   &AerodynamicDrag(
    Aerodynamic Drag=0.02
    Aerodynamic Rotational Drag=0
    Lift Contribution=0
    Pivot Offset Maximum:Vector3f=<1 0.5 0>
    Pivot Offset Minimum:Vector3f=<-1 -0.5 -0.001>
    Rotation Strength=0
    Use Linear Drag=true
    Use Rotational Drag=true
   )
   &AlignSpriteToMeshOrientation(Mesh Orientation Relative Sprite Facing Vector:Vector3f=<0 1 0>, SetAlignment=true, SetFacing=true)
   &SolveForcesAndVelocity(
    Acceleration Limit:NiagaraFloat=$User.MaxVel
    Clamp Velocity=true
    Force:Vector3f=&Multiply_Vector(A=$Transient.PhysicsForce, B=<1 1 1>)
    Limit Acceleration=true
    Speed Limit:NiagaraFloat=$User.MaxVel
    Velocity:Vector3f=&Multiply_Vector(A=$Particles.Velocity, B=<1 1 1>)
   ) {module=SolveForcesAndVelocity002}
   /Niagara/Modules/Landscape/SampleLandscape(Height Offset=-100)
   &RegionMask(
    Offset:Vector3f=<0 0 200>
    Offset Space:ENiagaraCoordinateSpace=::Local
    Region Falloff=0.1
    Region Radius:NiagaraFloat=$User.BoxSize
    UseOffset=false
    UseRotation=false
   ) {module=RegionMask001}
   &ScaleMeshSize(Scale Factor=&VectorFromFloat(Value=&Multiply_Float(
      A=&Multiply_Float(
       A=&FloatFromCurve(
        FloatCurve=DataInterface() {propertyValues {Curve {keys [
            {interpMode=RCIM_Cubic, tangentMode=RCTM_User, time=0, value=0}
            <at 0.01 10>
            {interpMode=RCIM_Cubic, tangentMode=RCTM_User, time=0.7, value=10}
            {tangentMode=RCTM_User, time=1, value=0}
           ]}}}
        Scale Curve=0.75
       )
       B=&CameraFade(Camera Query:NiagaraDataInterfaceCamera=?, Fade Offset:NiagaraFloat=250, Fade Distance:NiagaraFloat=250, Use Owner Position:NiagaraBool=true)
      )
      B=$Output.RegionMask001.RegionMask
     ))) {enabled=false}
   &ScaleMeshSize(Scale Factor=&VectorFromFloat(Value=&FloatFromCurve(
      FloatCurve=DataInterface() {propertyValues {Curve {
         defaultValue=1
         keys [
          {arriveTangent=6.6706, interpMode=RCIM_Cubic, leaveTangent=6.6706, tangentMode=RCTM_User, time=0, value=0}
          {interpMode=RCIM_Cubic, tangentMode=RCTM_User, time=0.5, value=1}
          {arriveTangent=-6.8957, tangentMode=RCTM_User, time=1, value=0}
         ]
        }}}
      Scale Curve=&Multiply_Float(A=1, B=$Particles.RegionMask001.Float)
     ))) {module=ScaleMeshSize001}
   &ScaleColor(
    Scale Alpha=&ClampFloat(Float=&FloatFromCurve(FloatCurve=DataInterface() {propertyValues {Curve {keys [<at 0.1 0>,<at 0.6 1>]}}}, CurveIndex=$Output.RegionMask001.RegionMask))
    Scale RGB=&Multiply_Vector(
     A=&Lerp_Vector(A=<1 1 1>, B=<1 1 1>, Alpha=$Output.RegionMask001.RegionMask)
     B=&VectorFromFloat(Value=&ClampFloat(Float=$Output.RegionMask001.RegionMask, Min=1))
    )
    ScaleA=true
    ScaleRGB=true
   ) {enabled=false}
   &KillParticlesInVolume(Invert Volume=true, Kill Shape=::NewEnumerator2, Origin Offset:Vector3f=<0 0 50>) {enabled=false}
  ]
  rendererClasses [&NiagaraMeshRendererProperties]
  renderers [{rendererClass=&NiagaraMeshRendererProperties, rendererIndex=0}]
 }
}
summary {
 systemName=NS_Env_TropicalFish_FishSchools_Tank
 emitters [
  {emitterName=Fish_HighEpic001, simTarget=GPUComputeSim, rendererClasses [&NiagaraMeshRendererProperties]}
  {
   emitterName=FN_WaterBubbles
   rendererClasses [&NiagaraMeshRendererProperties, &NiagaraSpriteRendererProperties]
  }
 ]
 userVariables [
  User.BoxSize {type=NiagaraFloat, defaultValue=800, description=""}
  User.CylinderRadius {type=NiagaraFloat, defaultValue=1000, description=""}
  User.MaxVel {type=NiagaraFloat, defaultValue=50, description=""}
  User.MeshScaleMult {type=NiagaraFloat, defaultValue=1, description=""}
  User.SpawnCap {type=NiagaraInt32, defaultValue=250}
  User.SpawnRate {type=NiagaraFloat, defaultValue=2500, description=""}
 ]
}
